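# Editor utility widget that forms one slot row in the Character Parts Anim Setup Tool, pairing a character part with a Control Rig Blueprint to add.
# On Construct, two Single Property Views bind to the widget itself, exposing its CharacterPart and Control_Rig_To_Add variables as editable fields.
# The ClearRow button resets both variables to null.
# The slot name text, SlotIndex and ActorLabel are not set in this graph.
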EditorUtilityWidgetBlueprint /AnimInEdTools/CharacterPartsAnimSetupTool/WBP_CPRow
graphs [EventGraph {nodes [
   "On Clicked (ClearRow)"() {is=title K2Node_ComponentBoundEvent}
   > Set CharacterPart(CharacterPart:Custom Character Part Object Reference) -> (Output_Get:Custom Character Part Object Reference)
   > Set Control_Rig_To_Add(Control_Rig_To_Add:Control Rig Blueprint Object Reference) -> (Output_Get:Control Rig Blueprint Object Reference)
   Event Construct()
   > SetObject(
    self=[
     Get CP_SinglePropertyView() -> (CP_SinglePropertyView:Single Property View Object Reference)
     Get ControlRig_SinglePropView() -> (ControlRig_SinglePropView:Single Property View Object Reference)
    ]
    NewObject=Self-Reference()
   )
  ]}]
variables [
 ActorLabel {type=string, default=""}
 CharacterPart {type=CustomCharacterPart, default=null}
 Control_Rig_To_Add {type=ControlRigBlueprint, default=null, class=~}
 SlotIndex {type=integer}
]
widgets {
 info {rootWidgetClass=&HorizontalBox, widgetCount=5}
 widgets [
  HorizontalBox_37 {widgetClassPath=&HorizontalBox, widget=HorizontalBox_37}
  SlotName_Text {
   widgetClassPath=&TextBlock
   bIsVariable=true
   parent=HorizontalBox_37
   widget=SlotName_Text
   properties {
    text=Text Block
    font {
     bForceMonospaced=false
     bMaterialIsStencil=false
     fontMaterial=null
     fontObject=&Roboto
     letterSpacing=0
     monospacedWidth=1
     size=10.5
     skewAmount=0
     typefaceFontName=Regular
     outlineSettings {
      bApplyOutlineToDropShadows=false
      bMiteredCorners=false
      bSeparateFillAlpha=false
      outlineMaterial=null
      outlineSize=0
      outlineColor=<rgba 0 0 0 1>
     }
    }
   }
   slot {class=HorizontalBoxSlot, properties {horizontalAlignment=HAlign_Left, verticalAlignment=VAlign_Center, size {sizeRule=Fill, value=0.18}}}
  }
  CP_SinglePropertyView {
   widgetClassPath=&SinglePropertyView
   parent=HorizontalBox_37
   widget=CP_SinglePropertyView
   properties {nameOverride=" ", propertyName=CharacterPart}
   slot {class=HorizontalBoxSlot, properties {verticalAlignment=VAlign_Center, size {sizeRule=Fill, value=0.42}}}
  }
  ControlRig_SinglePropView {
   widgetClassPath=&SinglePropertyView
   parent=HorizontalBox_37
   widget=ControlRig_SinglePropView
   properties {nameOverride=" ", propertyName=Control_Rig_To_Add}
   slot {class=HorizontalBoxSlot, properties {verticalAlignment=VAlign_Center, size {sizeRule=Fill, value=0.38}}}
  }
  ClearRow {
   widgetClassPath=&Button
   parent=HorizontalBox_37
   widget=ClearRow
   properties {
    toolTipText=Clear slot row
    widgetStyle {
     clickedSlateSound {resourceObject=null}
     disabled {
      bIsDynamicallyLoaded=false
      drawAs=RoundedBox
      imageType=NoImage
      mirroring=NoMirror
      resourceName=None
      resourceObject=null
      tiling=NoTile
      imageSize=<0 0>
      margin=<ltrb 0 0 0 0>
      outlineSettings {
       bUseBrushTransparency=false
       roundingType=FixedRadius
       width=1
       color {colorUseRule=UseColor_ColorTable, specifiedColor=<rgba 0 0 0 0>}
       cornerRadii=<4 4 4 4>
      }
      tintColor {colorUseRule=UseColor_Specified, specifiedColor=<rgba 0.0395 0.0395 0.0395 1>}
      uVRegion {bIsValid=false, max=<0 0>, min=<0 0>}
     }
     disabledForeground {colorUseRule=UseColor_ColorTable, specifiedColor=<rgba 0 0 0 0>}
     hovered {
      bIsDynamicallyLoaded=false
      drawAs=RoundedBox
      imageType=FullColor
      mirroring=NoMirror
      resourceName=None
      resourceObject=/Game/UI/Foundation/Textures/Buttons/T-SlotIcon-Remove-64
      tiling=NoTile
      imageSize=<20 20>
      margin=<ltrb 0 0 0 0>
      outlineSettings {
       bUseBrushTransparency=false
       roundingType=FixedRadius
       width=0
       color {colorUseRule=UseColor_ColorTable, specifiedColor=<rgba 0 0 0 0>}
       cornerRadii=<4 4 4 4>
      }
      tintColor {colorUseRule=UseColor_Specified, specifiedColor=<rgba 0.0953 0.0953 0.0953 1>}
      uVRegion {bIsValid=false, max=<0 0>, min=<0 0>}
     }
     hoveredForeground {colorUseRule=UseColor_ColorTable, specifiedColor=<rgba 0 0 0 0>}
     hoveredSlateSound {resourceObject=null}
     normal {
      bIsDynamicallyLoaded=false
      drawAs=Image
      imageType=FullColor
      mirroring=NoMirror
      resourceName=None
      resourceObject=&T-SlotIcon-Remove-64
      tiling=NoTile
      imageSize=<20 20>
      margin=<ltrb 0 0 0 0>
      outlineSettings {
       bUseBrushTransparency=false
       roundingType=FixedRadius
       width=0
       color {colorUseRule=UseColor_ColorTable, specifiedColor=<rgba 0 0 0 0>}
       cornerRadii=<4 4 4 4>
      }
      tintColor {colorUseRule=UseColor_Specified, specifiedColor=<rgba 1 1 1 1>}
      uVRegion {bIsValid=false, max=<0 0>, min=<0 0>}
     }
     normalForeground {colorUseRule=UseColor_ColorTable, specifiedColor=<rgba 0 0 0 1>}
     normalPadding=<ltrb 6 1.5 6 1.5>
     pressed {
      bIsDynamicallyLoaded=false
      drawAs=RoundedBox
      imageType=FullColor
      mirroring=NoMirror
      resourceName=None
      resourceObject=&T-SlotIcon-Remove-64
      tiling=NoTile
      imageSize=<20 20>
      margin=<ltrb 0 0 0 0>
      outlineSettings {
       bUseBrushTransparency=false
       roundingType=FixedRadius
       width=0
       color {colorUseRule=UseColor_ColorTable, specifiedColor=<rgba 0 0 0 0>}
       cornerRadii=<4 4 4 4>
      }
      tintColor {colorUseRule=UseColor_Specified, specifiedColor=<rgba 0.0284 0.0284 0.0284 1>}
      uVRegion {bIsValid=false, max=<0 0>, min=<0 0>}
     }
     pressedForeground {colorUseRule=UseColor_ColorTable, specifiedColor=<rgba 0 0 0 0>}
     pressedPadding=<ltrb 6 2.5 6 0.5>
     pressedSlateSound {resourceObject=null}
    }
   }
   slot {class=HorizontalBoxSlot, properties {horizontalAlignment=HAlign_Center, verticalAlignment=VAlign_Center, padding=<ltrb 0 0 10 0>}}
  }
 ]
}
