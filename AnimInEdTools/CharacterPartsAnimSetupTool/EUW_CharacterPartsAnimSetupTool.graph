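# Editor utility widget that sets up Fortnite character parts and Control Rigs for animating actors in Sequencer.
# It scans the focused Level Sequence, lists actors with character parts, and builds a WBP_CPRow per slot.
# Setup confirms empty slots, then chains embedded Python to swap parts on template and live actor, with a delay after removal.
# Python also adds layered rig tracks on components, regular ones on spawned "_Export" meshes, and prunes stale bindings.
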EditorUtilityWidgetBlueprint /AnimInEdTools/CharacterPartsAnimSetupTool/EUW_CharacterPartsAnimSetupTool
graphs [
 EventGraph {nodes [
   "On Clicked (EditorUtilityButton_194)"() {is=title K2Node_ComponentBoundEvent}
   > Execute Python Script(PythonScript="import webbrowser\r\n\r\nwebbrowser.open(\"https://epicgames.atlassian.net/wiki/spaces/FORT/pages/612861081/Character+Parts+Anim+Setup+Tool\")")
   "On Clicked (Setup_Button)"() {is=title K2Node_ComponentBoundEvent}
   > Clear(TargetArray=Get RemovedSlotNames() -> (RemovedSlotNames:Array of Strings))
   > Set RowAccum(RowAccum:String) -> (Output_Get:String)
   > n1 = Sequence()
   > n2 = For Each Loop(Array:Array of Widget Object References=GetAllChildren(self=Get Setup_ScrollBox() -> (Setup_ScrollBox:Editor Utility Scroll Box Object Reference))) -> (Array Element:Widget Object Reference)
   > n3 = Cast To WBP_CPRow!(execute, Object=@n2) -> (then, CastFailed, AsWBP CPRow:WBP CPRow Object Reference) {is=title K2Node_DynamicCast}
   n4 = Get CharacterPart!(self:WBP CPRow Object Reference=@n3) -> (CharacterPart:Custom Character Part Object Reference)
   > n5 = Is Valid(InputObject=@n4)
   n6 = Set Character_Part_Path(execute=@n5.Is Not Valid, Character_Part_Path:String=(none)) -> (Output_Get:String)
   n7 = Get Control_Rig_To_Add!(self:WBP CPRow Object Reference=@n3) -> (Control_Rig_To_Add:Control Rig Blueprint Object Reference)
   n8 = Is Valid(exec=[@n14.then, @n6.then], InputObject=@n7)
   n9 = Set ControlRig_Path(execute=@n8.Is Not Valid, ControlRig_Path:String=(none)) -> (Output_Get:String)
   n10 = Get SlotIndex!(self:WBP CPRow Object Reference=@n3) -> (SlotIndex:int=0)
   Set RowAccum(
    execute=[@n9.then, @n12.then]
    RowAccum=Append!(
     A=Get RowAccum() -> (RowAccum:String)
     B=|||
     C:String="To String (Integer)"(InInt=@n10)
     D:String=||
     E:String=Get Character_Part_Path() -> (Character_Part_Path:String)
     F:String=||
     G:String=Get ControlRig_Path() -> (ControlRig_Path:String)
    ) -> (ReturnValue)
   )
   n11 = GetPathNameForLoadedAsset(execute=@n8.Is Valid, LoadedAsset=@n7)
   > n12 = Set ControlRig_Path(ControlRig_Path=@n11)
   n13 = GetPathNameForLoadedAsset(execute=@n5.Is Valid, LoadedAsset=@n4)
   > n14 = Set Character_Part_Path(Character_Part_Path=@n13)
   n15 = For Each Loop(Exec=@n1.then_1, Array:Array of Strings=ParseIntoArray(SourceString=Get RowAccum(), Delimiter=|||)) -> (Array Element:String)
   n16 = Branch(execute=@n15.Completed, Condition=Is Empty!(TargetArray=Get RemovedSlotNames()) -> (ReturnValue:bool))
   Execute Python Script!(
    execute=[@n19.then, @n16.then]
    PythonScript="import unreal\r\n\r\nPART_TYPE_MAP = {\r\n  0: unreal.FortCustomPartType.HEAD,\r\n  1: unreal.FortCustomPartType.BODY,\r\n  2: unreal.FortCustomPartType.HAT,\r\n  3: unreal.FortCustomPartType.BACKPACK,\r\n  4: unreal.FortCustomPartType.MISC_OR_TAIL,\r\n  5: unreal.FortCustomPartType.FACE,\r\n  6: unreal.FortCustomPartType.GAMEPLAY,\r\n  7: unreal.FortCustomPartType.EXTRA_PART,\r\n  8: unreal.FortCustomPartType.GAMEPLAY2,\r\n}\r\n\r\nseq = unreal.LevelSequenceEditorBlueprintLibrary.get_current_level_sequence()\r\nlses = unreal.get_editor_subsystem(unreal.LevelSequenceEditorSubsystem)\r\nworld = unreal.get_editor_subsystem(unreal.UnrealEditorSubsystem).get_editor_world()\r\n\r\ntemplate = None\r\nfor b in seq.get_bindings():\r\n  if actor_label not in str(unreal.MovieSceneBindingExtensions.get_display_name(b)):\r\n      continue\r\n  cb = lses.get_custom_binding_objects(b)\r\n  if cb:\r\n      template = cb[0].get_editor_property(\"actor_template\")\r\n  break\r\n\r\nif not template:\r\n  print(f\"[CPTool] template not found: {actor_label}\")\r\nelse:\r\n  current_parts = list(template.get_editor_property(\"character_part_list\"))\r\n  live_actor = None\r\n  for actor in unreal.GameplayStatics.get_all_actors_of_class(world, unreal.Actor):\r\n      if actor_label in actor.get_actor_label():\r\n          live_actor = actor\r\n          break\r\n  if not live_actor:\r\n      print(f\"[CPTool] live actor not found: {actor_label}\")\r\n  else:\r\n      for row in row_updates.split(\"|||\"):\r\n          fields = [f.strip('\"') for f in row.split(\"||\")]\r\n          if len(fields) < 3:\r\n              continue\r\n          slot_idx = int(fields[0])\r\n          new_cp_path = fields[1]\r\n          current_cp = current_parts[slot_idx] if slot_idx < len(current_parts) else None\r\n          current_cp_path = current_cp.get_path_name().split(\":\")[0].strip() if current_cp else \"(none)\"\r\n          if new_cp_path == current_cp_path:\r\n              continue\r\n          try:\r\n              live_actor.remove_part(PART_TYPE_MAP[slot_idx])\r\n              print(f\"[CPTool] removed slot {slot_idx}\")\r\n          except Exception as e:\r\n              print(f\"[CPTool] remove_part error slot {slot_idx}: {e}\")"
    actor_label:String=Get Selected_Actor() -> (Selected_Actor:String)
    row_updates:String=Get RowAccum()
   ) -> (then, ReturnValue) {inputs [actor_label,row_updates]}
   > Delay!(execute, WorldContextObject:Object Reference, Duration=1.5) -> (then)
   > Execute Python Script(
    PythonScript="import unreal\r\n\r\nseq = unreal.LevelSequenceEditorBlueprintLibrary.get_current_level_sequence()\r\nlses = unreal.get_editor_subsystem(unreal.LevelSequenceEditorSubsystem)\r\nworld = unreal.get_editor_subsystem(unreal.UnrealEditorSubsystem).get_editor_world()\r\n\r\ntemplate = None\r\nfor b in seq.get_bindings():\r\n  if actor_label not in str(unreal.MovieSceneBindingExtensions.get_display_name(b)):\r\n      continue\r\n  cb = lses.get_custom_binding_objects(b)\r\n  if cb:\r\n      template = cb[0].get_editor_property(\"actor_template\")\r\n  break\r\n\r\nif not template:\r\n  print(f\"[CPTool] template not found: {actor_label}\")\r\nelse:\r\n  current_parts = list(template.get_editor_property(\"character_part_list\"))\r\n  template_parts = list(current_parts)\r\n  live_actor = None\r\n  for actor in unreal.GameplayStatics.get_all_actors_of_class(world, unreal.Actor):\r\n      if actor_label in actor.get_actor_label():\r\n          live_actor = actor\r\n          break\r\n  if not live_actor:\r\n      print(f\"[CPTool] live actor not found: {actor_label}\")\r\n  else:\r\n      live_parts = list(live_actor.get_editor_property(\"character_part_list\"))\r\n      dirty = False\r\n      for row in row_updates.split(\"|||\"):\r\n          fields = [f.strip('\"') for f in row.split(\"||\")]\r\n          if len(fields) < 3:\r\n              continue\r\n          slot_idx = int(fields[0])\r\n          new_cp_path = fields[1]\r\n          current_cp = current_parts[slot_idx] if slot_idx < len(current_parts) else None\r\n          current_cp_path = current_cp.get_path_name().split(\":\")[0].strip() if current_cp else \"(none)\"\r\n          if new_cp_path == current_cp_path:\r\n              continue\r\n          new_cp = None if new_cp_path == \"(none)\" else unreal.load_asset(new_cp_path)\r\n          template_parts[slot_idx] = new_cp\r\n          live_parts[slot_idx] = new_cp\r\n          dirty = True\r\n          print(f\"[CPTool] queued slot {slot_idx} = {new_cp_path}\")\r\n      if dirty:\r\n          template.modify()\r\n          template.set_editor_property(\"character_part_list\", template_parts,\r\n              notify_mode=unreal.PropertyAccessChangeNotifyMode.NEVER)\r\n          unreal.EditorAssetLibrary.save_loaded_asset(seq, only_if_is_dirty=False)\r\n          live_actor.modify()\r\n          live_actor.set_editor_property(\"character_part_list\", live_parts)\r\n          print(f\"[CPTool] applied all CP changes to {actor_label}\")"
    actor_label=Get Selected_Actor()
    row_updates=Get RowAccum()
   ) {inputs [actor_label,row_updates]}
   > Execute Python Script(
    PythonScript="import unreal, tempfile, os, re\r\n\r\nSLOT_COMPONENT_MAP = {\r\n  0: \"Head\", 1: \"Body\", 2: \"Hat\", 3: \"Backpack\", 4: \"MiscOrTail\",\r\n  5: \"Face\", 6: \"GameplayOne\", 7: \"ExtraPart\", 8: \"GameplayTwo\",\r\n}\r\n\r\ndef get_cp_mesh(cp_asset):\r\n  mesh = cp_asset.get_editor_property(\"skeletal_mesh\")\r\n  if mesh and mesh.get_name() != \"Hat_Empty\":\r\n      return mesh\r\n  tmp = os.path.join(tempfile.gettempdir(), \"_cp_mesh_check.T3D\")\r\n  task = unreal.AssetExportTask()\r\n  task.set_editor_property(\"object\", cp_asset)\r\n  task.set_editor_property(\"filename\", tmp)\r\n  task.set_editor_property(\"automated\", True)\r\n  task.set_editor_property(\"replace_identical\", True)\r\n  unreal.Exporter.run_asset_export_task(task)\r\n  with open(tmp, \"r\", encoding=\"utf-8\") as f:\r\n      text = f.read()\r\n  m = re.search(r'SkeletalMeshParameters=\\([^)]*ParameterValue=\"([^\"]+)\"', text)\r\n  if m:\r\n      return unreal.load_asset(m.group(1))\r\n  return None\r\n\r\nseq = unreal.LevelSequenceEditorBlueprintLibrary.get_current_level_sequence()\r\nlses = unreal.get_editor_subsystem(unreal.LevelSequenceEditorSubsystem)\r\nworld = unreal.get_editor_subsystem(unreal.UnrealEditorSubsystem).get_editor_world()\r\n\r\nactor_binding = None\r\nfor b in seq.get_bindings():\r\n  if actor_label in str(unreal.MovieSceneBindingExtensions.get_display_name(b)):\r\n      actor_binding = b\r\n      break\r\n\r\nif not actor_binding:\r\n  print(f\"[CPTool] actor binding not found: {actor_label}\")\r\nelse:\r\n  live_actor = None\r\n  for actor in unreal.GameplayStatics.get_all_actors_of_class(world, unreal.Actor):\r\n      if actor_label in actor.get_actor_label():\r\n          live_actor = actor\r\n          break\r\n\r\n  for row in row_updates.split(\"|||\"):\r\n      fields = [f.strip('\"') for f in row.split(\"||\")]\r\n      if len(fields) < 3:\r\n          continue\r\n      slot_idx = int(fields[0])\r\n      cp_path = fields[1]\r\n      cr_path = fields[2]\r\n\r\n      if not cr_path or cr_path == \"(none)\":\r\n          continue\r\n\r\n      component_name = SLOT_COMPONENT_MAP.get(slot_idx)\r\n      if not component_name:\r\n          continue\r\n\r\n      cr_blueprint = unreal.load_asset(cr_path)\r\n      if not cr_blueprint:\r\n          print(f\"[CPTool] could not load CR: {cr_path}\")\r\n          continue\r\n      cr_class = cr_blueprint.generated_class()\r\n\r\n      ## Find or create component binding\r\n      comp_binding = None\r\n      for b in seq.get_bindings():\r\n          if str(unreal.MovieSceneBindingExtensions.get_display_name(b)) == component_name:\r\n              comp_binding = b\r\n              break\r\n      if not comp_binding and live_actor:\r\n          comps = live_actor.get_components_by_class(unreal.SkeletalMeshComponentBudgeted)\r\n          comp = next((c for c in comps if c.get_name() == component_name), None)\r\n          if comp:\r\n              comp_binding = seq.add_possessable(comp)\r\n              unreal.MovieSceneBindingExtensions.set_parent(comp_binding, actor_binding)\r\n              print(f\"[CPTool] added possessable: {component_name}\")\r\n          else:\r\n              print(f\"[CPTool] component not found: {component_name}\")\r\n              continue\r\n\r\n      ## Layered CR on the component\r\n      if comp_binding:\r\n          track = unreal.ControlRigSequencerLibrary.find_or_create_control_rig_track(\r\n              world, seq, cr_class, comp_binding, is_layered_control_rig=True)\r\n          if track:\r\n              print(f\"[CPTool] layered CR track added for slot {slot_idx}\")\r\n          else:\r\n              print(f\"[CPTool] failed to add layered CR track for slot {slot_idx}\")\r\n\r\n      ## Standalone skeletal mesh with regular CR (only when CP is also set)\r\n      if cp_path and cp_path != \"(none)\":\r\n          cp_asset = unreal.load_asset(cp_path)\r\n          if not cp_asset:\r\n              print(f\"[CPTool] could not load CP: {cp_path}\")\r\n              continue\r\n          mesh = get_cp_mesh(cp_asset)\r\n          if not mesh:\r\n              print(f\"[CPTool] no skeletal mesh found for CP: {cp_path}\")\r\n              continue\r\n\r\n          export_label = f\"{mesh.get_name()}_Export\"\r\n\r\n          standalone_binding = None\r\n          for b in seq.get_bindings():\r\n              if str(unreal.MovieSceneBindingExtensions.get_display_name(b)) == export_label:\r\n                  standalone_binding = b\r\n                  break\r\n\r\n          if not standalone_binding:\r\n              standalone_actor = unreal.EditorLevelLibrary.spawn_actor_from_class(\r\n                  unreal.SkeletalMeshActor, unreal.Vector(0, 0, 0), unreal.Rotator(0, 0, 0))\r\n              if not standalone_actor:\r\n                  print(f\"[CPTool] failed to spawn standalone actor for slot {slot_idx}\")\r\n                  continue\r\n              standalone_actor.set_actor_label(export_label)\r\n              smc = standalone_actor.get_editor_property(\"skeletal_mesh_component\")\r\n              if smc:\r\n                  smc.set_skinned_asset_and_update(mesh)\r\n              standalone_binding = seq.add_possessable(standalone_actor)\r\n              print(f\"[CPTool] spawned {export_label} with mesh {mesh.get_name()}\")\r\n\r\n          ## Regular (non-layered) CR on the standalone mesh\r\n          track = unreal.ControlRigSequencerLibrary.find_or_create_control_rig_track(\r\n              world, seq, cr_class, standalone_binding, is_layered_control_rig=False)\r\n          if track:\r\n              print(f\"[CPTool] regular CR track added for {export_label}\")\r\n          else:\r\n              print(f\"[CPTool] failed to add regular CR track for {export_label}\")\r\n\r\n  unreal.EditorAssetLibrary.save_loaded_asset(seq, only_if_is_dirty=False)"
    actor_label=Get Selected_Actor()
    row_updates=Get RowAccum()
   ) {inputs [actor_label,row_updates]}
   > Execute Python Script!(
    execute
    PythonScript="import unreal, os, tempfile, re\r\n\r\nlevel_sequence = unreal.LevelSequenceEditorBlueprintLibrary.get_current_level_sequence()\r\nlevel_sequence_editor_sub = unreal.get_editor_subsystem(unreal.LevelSequenceEditorSubsystem)\r\nworld = unreal.get_editor_subsystem(unreal.UnrealEditorSubsystem).get_editor_world()\r\n\r\nSLOT_COMPONENT_MAP = {\r\n  0: \"Head\", 1: \"Body\", 2: \"Hat\", 3: \"Backpack\", 4: \"MiscOrTail\",\r\n  5: \"Face\", 6: \"GameplayOne\", 7: \"ExtraPart\", 8: \"GameplayTwo\",\r\n}\r\n\r\nactor_template = None\r\nfor binding in level_sequence.get_bindings():\r\n  if actor_label not in str(unreal.MovieSceneBindingExtensions.get_display_name(binding)):\r\n      continue\r\n  character_binding = level_sequence_editor_sub.get_custom_binding_objects(binding)\r\n  if character_binding:\r\n      actor_template = character_binding[0].get_editor_property(\"actor_template\")\r\n  break\r\n\r\nif not actor_template:\r\n  print(f\"[CPTool] template not found: {actor_label}\")\r\nelse:\r\n  current_parts = list(actor_template.get_editor_property(\"character_part_list\"))\r\n  live_actor = None\r\n  for actor in unreal.GameplayStatics.get_all_actors_of_class(world, unreal.Actor):\r\n      if actor_label in actor.get_actor_label():\r\n          live_actor = actor\r\n          break\r\n  if not live_actor:\r\n      print(f\"[CPTool] live actor not found: {actor_label}\")\r\n  else:\r\n      expected_export_labels = set()\r\n      for slot_index, character_part in enumerate(current_parts):\r\n          if character_part is None:\r\n              continue\r\n          mesh = character_part.get_editor_property(\"skeletal_mesh\")\r\n          if mesh and mesh.get_name() != \"Hat_Empty\":\r\n              expected_export_labels.add(f\"{mesh.get_name()}_Export\")\r\n          else:\r\n              tmp = os.path.join(tempfile.gettempdir(), \"_character_part_check.T3D\")\r\n              task = unreal.AssetExportTask()\r\n              task.set_editor_property(\"object\", character_part)\r\n              task.set_editor_property(\"filename\", tmp)\r\n              task.set_editor_property(\"automated\", True)\r\n              task.set_editor_property(\"replace_identical\", True)\r\n              unreal.Exporter.run_asset_export_task(task)\r\n              with open(tmp, \"r\", encoding=\"utf-8\") as f:\r\n                  text = f.read()\r\n              m = re.search(r'SkeletalMeshParameters=\\([^)]*ParameterValue=\"([^\"]+)\"', text)\r\n              if m:\r\n                  mesh = unreal.load_asset(m.group(1))\r\n                  if mesh:\r\n                      expected_export_labels.add(f\"{mesh.get_name()}_Export\")\r\n\r\n      for binding in level_sequence.get_bindings():\r\n          label = str(unreal.MovieSceneBindingExtensions.get_display_name(binding))\r\n          if not label.endswith(\"_Export\"):\r\n              continue\r\n          if label not in expected_export_labels:\r\n              binding.remove()\r\n              for actor in unreal.GameplayStatics.get_all_actors_of_class(world, unreal.Actor):\r\n                  if actor.get_actor_label() == label:\r\n                      actor.destroy_actor()\r\n                      break\r\n\r\n      for slot, character_part in enumerate(current_parts):\r\n          if character_part is None:\r\n              component_name = SLOT_COMPONENT_MAP.get(slot)\r\n              for binding in level_sequence.get_bindings():\r\n                  if component_name == str(unreal.MovieSceneBindingExtensions.get_display_name(binding)):\r\n                      for track in unreal.MovieSceneBindingExtensions.get_tracks(binding):\r\n                          binding.remove_track(track)\r\n                      binding.remove()\r\n                      break\r\n\r\n      unreal.EditorAssetLibrary.save_loaded_asset(level_sequence, only_if_is_dirty=False)"
    actor_label=Get Selected_Actor()
   ) -> (then, ReturnValue) {inputs [actor_label]}
   n17 = JoinStringArray(SourceArray=Get RemovedSlotNames(), Separator="\r\n")
   n18 = Show Message Dialog(
    execute=@n16.else
    Title="To Text (String)"(InString=Character Setup Tool)
    Message="To Text (String)"(InString=Append!(
      A="The following slots will be cleared from the scene: \r\n\r\n"
      B=@n17
      C="\r\n\r\nIf animations are attached to these tracks, they will be removed from the scene. Proceed?"
     ) -> (ReturnValue))
    MessageType=YesNo
    DefaultValue=Yes
    MessageCategory=Warning
   )
   > n19 = Branch(Condition="Equal (Enum)"(A:EAppReturnType Enum=@n18, B:EAppReturnType Enum=Yes))
   n20 = ParseIntoArray(SourceString=@n15, Delimiter=||)
   Branch(
    execute=@n15.LoopBody
    Condition="Equal Exactly (String)"(A="Get (a ref)"(Array:Array of Strings=@n20, Dimension 1=1) -> (Output:"String (by ref)"), B=(none))
   )
   > n21 = Switch on String!(execute:Exec, Selection:String="Get (a copy)"(Array=@n20)) -> (Default:Exec, 0:Exec, 1:Exec, 2:Exec, 3:Exec, 4:Exec, 5:Exec, 6:Exec, 7:Exec, 8:Exec) {pinNames [0,1,2,3,4,5,6,7,8]}
   Add(execute=@n21.0, TargetArray=Get RemovedSlotNames(), NewItem=Head) {is=title K2Node_CallArrayFunction}
   Add(execute=@n21.1, TargetArray=Get RemovedSlotNames(), NewItem=Body)
   Add(execute=@n21.2, TargetArray=Get RemovedSlotNames(), NewItem=Hat)
   Add(execute=@n21.3, TargetArray=Get RemovedSlotNames(), NewItem=Backpack)
   Add(execute=@n21.4, TargetArray=Get RemovedSlotNames(), NewItem=MiscOrTail)
   Add(execute=@n21.5, TargetArray=Get RemovedSlotNames(), NewItem=Face)
   Add(execute=@n21.6, TargetArray=Get RemovedSlotNames(), NewItem=GameplayOne)
   Add(execute=@n21.7, TargetArray=Get RemovedSlotNames(), NewItem=ExtraPart)
   Add(execute=@n21.8, TargetArray=Get RemovedSlotNames(), NewItem=GameplayTwo)
   n22 = "On Selection Changed (Actor_with_CharacterParts_ComboBox)"!() -> (OutputDelegate, then, SelectedItem:String, SelectionType:ESelectInfo Enum) {is=title K2Node_ComponentBoundEvent}
   > Set Selected_Actor(Selected_Actor:String=@n22.SelectedItem) -> (Output_Get:String)
   > PopulateCharacterPartRows()
   n23 = "On Clicked (Scan_Btn)"() {is=title K2Node_ComponentBoundEvent}
   Update Character Part and Actors(execute=[@n23.then, @n24.then])
   n24 = Event Construct()
  ]}
 Populate Actors with CharacterParts {nodes [
   Populate Actors with CharacterParts() {is=title K2Node_FunctionEntry}
   > n1 = Execute Python Script!(
    execute
    PythonScript="import unreal\r\n\r\nworld = unreal.get_editor_subsystem(unreal.UnrealEditorSubsystem).get_editor_world()\r\nactor_labels = []\r\n\r\nif level_sequence:\r\n  bindings = level_sequence.get_bindings()\r\n  playback_range = level_sequence.get_playback_range()\r\n  bound_results = unreal.SequencerTools.get_bound_objects(\r\n      world, level_sequence, bindings, playback_range)\r\n  for result, binding in zip(bound_results, bindings):\r\n      for obj in result.bound_objects:\r\n          try:\r\n              parts = obj.get_editor_property(\"character_part_list\")\r\n              if parts is not None:\r\n                  label = str(unreal.MovieSceneBindingExtensions.get_display_name(binding))\r\n                  actor_labels.append(label)\r\n                  break\r\n          except:\r\n              pass\r\nelse:\r\n  print(\"[CPTool] no level sequence open\")\r\n\r\nactor_list = \"|||\".join(actor_labels)"
    level_sequence:Level Sequence Object Reference=Get Current_LevelSequence() -> (Current_LevelSequence:Level Sequence Object Reference)
   ) -> (then, ReturnValue, actor_list:String) {inputs [level_sequence], outputs [actor_list]}
   > n2 = Branch(Condition="Equal Exactly (String)"(A=@n1.actor_list, B=Get CacheActorList() -> (CacheActorList:String)))
   n3 = ParseIntoArray(SourceString=@n1.actor_list, Delimiter=|||)
   > n4 = Set Selected_Actor(Selected_Actor="Get (a ref)"(Array=@n3, Dimension 1=0))
   > PopulateCharacterPartRows()
   > SetSelectedOption!(
    execute
    self:"Searchable Combo Box (String) Object Reference"=Get Actor_with_CharacterParts_ComboBox() -> (Actor_with_CharacterParts_ComboBox:"Searchable Combo Box (String) Object Reference")
    Option:String=@n4
   ) -> (then)
   ClearOptions(execute=@n2.else, self:"Searchable Combo Box (String) Object Reference"=Get Actor_with_CharacterParts_ComboBox())
   > n5 = For Each Loop(Array=@n3)
   Set CacheActorList(execute=@n5.Completed, CacheActorList:String=@n1.actor_list) -> (Output_Get:String)
   > Set Selected_Actor(Selected_Actor="Get (a ref)"(Array=@n3, Dimension 1=0))
   > PopulateCharacterPartRows()
   > SetSelectedOption(self=Get Actor_with_CharacterParts_ComboBox(), Option=Get Selected_Actor())
   AddOption!(execute=@n5.LoopBody, self:"Searchable Combo Box (String) Object Reference"=Get Actor_with_CharacterParts_ComboBox(), Option:String=@n5) -> (then)
  ]}
 PopulateCharacterPartRows {nodes [
   PopulateCharacterPartRows() {is=title K2Node_FunctionEntry}
   > ClearChildren(self=Get Setup_ScrollBox())
   > Clear(TargetArray=Get ParsedArray() -> (ParsedArray:Array of Strings))
   > n1 = Sequence()
   > n2 = Execute Python Script!(
    execute
    PythonScript="import unreal\r\n\r\nslot_names = [str(e).split(\".\")[-1].split(\":\")[0].strip().replace(\"_\", \" \").title()\r\n            for e in unreal.FortCustomPartType]\r\nrows = []\r\n\r\nworld = unreal.get_editor_subsystem(unreal.UnrealEditorSubsystem).get_editor_world()\r\n\r\nif level_sequence:\r\n  bindings = level_sequence.get_bindings()\r\n  playback_range = level_sequence.get_playback_range()\r\n  bound_results = unreal.SequencerTools.get_bound_objects(\r\n      world, level_sequence, bindings, playback_range)\r\n  for result, binding in zip(bound_results, bindings):\r\n      actor_label = str(unreal.MovieSceneBindingExtensions.get_display_name(binding))\r\n      if actor_label != selected_actor:\r\n          continue\r\n      for obj in result.bound_objects:\r\n          try:\r\n              parts = obj.get_editor_property(\"character_part_list\")\r\n              if parts is None:\r\n                  continue\r\n              for i, part in enumerate(parts):\r\n                  cp_path = \"(none)\"\r\n                  if part is not None:\r\n                      try:\r\n                          cp_path = part.get_path_name().split(\":\")[0].strip() or \"(none)\"\r\n                      except Exception as e:\r\n                          print(f\"[CPTool] cp_path error: {e}\")\r\n                  slot = slot_names[i] if i < len(slot_names) else str(i)\r\n                  rows.append(f\"{slot}||{i}||{cp_path}\")\r\n          except Exception as e:\r\n              print(f\"[CPTool] skipped obj: {e}\")\r\nelse:\r\n  print(\"[CPTool] no level sequence open\")\r\n\r\nfor r in rows:\r\n  print(f\"[CPTool] row: {r}\")\r\nprint(f\"[CPTool] found {len(rows)} part rows for {selected_actor}\")\r\nrow_data = \"|||\".join(rows)"
    level_sequence=Get Current_LevelSequence()
    selected_actor:String=Get Selected_Actor()
   ) -> (then, ReturnValue, row_data:String) {inputs [level_sequence, selected_actor], outputs [row_data]}
   > Set ParsedArray(ParsedArray:Array of Strings=ParseIntoArray(SourceString=@n2.row_data, Delimiter=|||)) -> (Output_Get:Array of Strings)
   > n3 = Branch(Condition=IsEmpty(InString=@n2.row_data))
   > Set Visibility(Visibility=Visible, self=Get NoParts_Text() -> (NoParts_Text:Text Object Reference))
   > Set Visibility(Visibility=Collapsed, self=Get Setup_ScrollBox())
   Set Visibility(execute=@n3.else, Visibility=Collapsed, self=Get NoParts_Text())
   > Set Visibility(Visibility=Visible, self=Get Setup_ScrollBox())
   n4 = For Each Loop(Exec=@n1.then_1, Array=Get ParsedArray())
   > n5 = Create Widget!(
    execute:Exec
    WorldContextObject:Object Reference
    Class:User Widget Class Reference=&/WBP_CPRow.WBP_CPRow_C
    OwningPlayer:Player Controller Object Reference
   ) -> (then:Exec, ReturnValue:WBP CPRow Object Reference)
   > n6 = Set CP_Row(CP_Row:WBP CPRow Object Reference=@n5) -> (Output_Get:WBP CPRow Object Reference)
   > AddChild(self=Get Setup_ScrollBox(), Content=@n6)
   n7 = ParseIntoArray(SourceString=@n4, Delimiter=||)
   > Set SlotIndex!(
    execute
    SlotIndex:int=String To Integer(InString="Get (a ref)"(Array=@n7, Dimension 1=1))
    self:WBP CPRow Object Reference=Get CP_Row() -> (CP_Row:WBP CPRow Object Reference)
   ) -> (then, Output_Get:int=0)
   > Set ActorLabel!(execute, ActorLabel:String=Get Selected_Actor(), self:WBP CPRow Object Reference=Get CP_Row()) -> (then, Output_Get:String)
   > Set Text(
    Text="To Text (String)"(InString="Get (a ref)"(Array=@n7, Dimension 1=0))
    self=Get SlotName_Text!(self:WBP CPRow Object Reference=Get CP_Row()) -> (SlotName_Text:Text Object Reference)
   )
   n8 = "Get (a ref)"(Array=@n7, Dimension 1=2)
   > Branch(Condition="Not Equal Exactly (String)"(A=@n8, B=(none)))
   > n9 = LoadAsset(AssetPath=MakeSoftObjectPath!(PathString:String=@n8) -> (ReturnValue_PathString:String))
   > n10 = Cast To CustomCharacterPart!(execute, Object=@n9) -> (then, CastFailed, AsCustom Character Part:Custom Character Part Object Reference) {is=title K2Node_DynamicCast}
   n11 = Get CP_Row()
   > Set CharacterPart!(execute, CharacterPart:Custom Character Part Object Reference=@n10, self:WBP CPRow Object Reference=@n11) -> (then, Output_Get:Custom Character Part Object Reference)
   > SetObject(
    self=Get CP_SinglePropertyView!(self:WBP CPRow Object Reference=@n11) -> (CP_SinglePropertyView:Single Property View Object Reference)
    NewObject=Get CP_Row()
   )
  ]}
 RefreshComboDisplay {nodes [
   RefreshComboDisplay() {is=title K2Node_FunctionEntry}
   > SetSelectedOption(self=Get Actor_with_CharacterParts_ComboBox(), Option=Get Selected_Actor())
  ]}
 Update Character Part and Actors {nodes [
   Update Character Part and Actors() {is=title K2Node_FunctionEntry}
   > n1 = GetFocusedLevelSequence!(execute) -> (then, ReturnValue:Level Sequence Object Reference)
   > n2 = Is Valid(InputObject=@n1)
   n3 = Set LevelSequenceName(execute=@n2.Is Not Valid, LevelSequenceName:String=No Level Sequence Detected) -> (Output_Get:String)
   n4 = Get LS_Name() -> (LS_Name:Text Object Reference)
   > Set Text(Text="To Text (String)"(InString=@n3), self=@n4)
   > Set Selected_Actor()
   > Set!(
    execute
    ColorAndOpacity_SpecifiedColor:Linear Color Structure=(R=1,G=0,B=0,A=1)
    ColorAndOpacity_ColorUseRule:ESlateColorStylingMode Enum
    self:Text Object Reference=@n4
   ) -> (then, Output_Get:Slate Color Structure)
   > ClearOptions(self=Get Actor_with_CharacterParts_ComboBox())
   > ClearChildren(self=Get Setup_ScrollBox())
   Set Current_LevelSequence(execute=@n2.Is Valid, Current_LevelSequence:Level Sequence Object Reference=@n1) -> (Output_Get:Level Sequence Object Reference)
   > Set LevelSequenceName(LevelSequenceName=GetDisplayName(Object=@n1))
   n5 = Get LS_Name()
   > Set Text(Text="To Text (String)"(InString=Get LevelSequenceName() -> (LevelSequenceName:String)), self=@n5)
   > Set(ColorAndOpacity_SpecifiedColor=(R=0,G=1,B=0,A=1), self=@n5)
   > Populate Actors with CharacterParts() {is=title K2Node_CallFunction}
  ]}
]
variables [
 CacheActorList {type=string, default=""}
 Character_Part_Path {type=string, default=""}
 Control_Rig_To_Add {type=ControlRigBlueprint, default=null, class=~}
 ControlRig_Path {type=string, default=""}
 Current_LevelSequence {type=LevelSequence, default=null}
 LevelSequenceName {type=string, default=""}
 RemovedSlotNames {type="array<string>", default []}
 RowAccum {type=string, default=""}
 Selected_Actor {type=string, default=""}
 SlotOptions {type="array<string>", default []}
]
widgets {
 info {rootWidgetClass=&CanvasPanel, widgetCount=26}
 widgets [
  CanvasPanel_30 {widgetClassPath=&CanvasPanel, widget=CanvasPanel_30}
  VerticalBox_1 {
   widgetClassPath=&VerticalBox
   parent=CanvasPanel_30
   widget=VerticalBox_1
   slot {class=CanvasPanelSlot, properties {layoutData {alignment=<0 0>, anchors {maximum=<1 1>, minimum=<0 0>}, offsets=<ltrb 0 0 0.9009 6.036>}}}
  }
  TextBlock_148 {
   widgetClassPath=&TextBlock
   parent=VerticalBox_1
   widget=TextBlock_148
   properties {text=Character Parts Anim Setup}
   slot {class=VerticalBoxSlot, properties {padding=<ltrb 6 6 6 6>}}
  }
  TextBlock {
   widgetClassPath=&TextBlock
   parent=VerticalBox_1
   widget=TextBlock
   properties {
    text="Setup and add character parts to selected actor in  Sequencer to animate"
    font {
     bForceMonospaced=false
     bMaterialIsStencil=false
     fontMaterial=null
     fontObject=&Roboto
     letterSpacing=0
     monospacedWidth=1
     size=10.5
     skewAmount=0
     typefaceFontName=Regular
     outlineSettings {
      bApplyOutlineToDropShadows=false
      bMiteredCorners=false
      bSeparateFillAlpha=false
      outlineMaterial=null
      outlineSize=0
      outlineColor=<rgba 0 0 0 1>
     }
    }
   }
   slot {class=VerticalBoxSlot, properties {padding=<ltrb 6 6 6 6>}}
  }
  Border_0 {
   widgetClassPath=&Border
   parent=VerticalBox_1
   widget=Border_0
   properties {brushColor=<rgba 0.0103 0.0103 0.0103 1>}
   slot {class=VerticalBoxSlot, properties {padding=<ltrb 10 10 10 10>}}
  }
  ControlRig {widgetClassPath=&VerticalBox, parent=Border_0, widget=ControlRig, slot {class=BorderSlot}}
  VerticalBox_0 {
   widgetClassPath=&VerticalBox
   parent=ControlRig
   widget=VerticalBox_0
   slot {class=VerticalBoxSlot, properties {padding=<ltrb 6 4 6 4>}}
  }
  ActiveSequence_HB {
   widgetClassPath=&HorizontalBox
   parent=VerticalBox_0
   widget=ActiveSequence_HB
   slot {class=VerticalBoxSlot, properties {padding=<ltrb 0 0 0 10>}}
  }
  TextBlock_73 {
   widgetClassPath=&TextBlock
   parent=ActiveSequence_HB
   widget=TextBlock_73
   properties {
    text=Active Level Sequence:
    font {
     bForceMonospaced=false
     bMaterialIsStencil=false
     fontMaterial=null
     fontObject=&Roboto
     letterSpacing=0
     monospacedWidth=1
     size=10.5
     skewAmount=0
     typefaceFontName=Regular
     outlineSettings {
      bApplyOutlineToDropShadows=false
      bMiteredCorners=false
      bSeparateFillAlpha=false
      outlineMaterial=null
      outlineSize=0
      outlineColor=<rgba 0 0 0 1>
     }
    }
   }
   slot {class=HorizontalBoxSlot, properties {verticalAlignment=VAlign_Center}}
  }
  LS_Name {
   widgetClassPath=&TextBlock
   bIsVariable=true
   parent=ActiveSequence_HB
   widget=LS_Name
   properties {
    text=Text Block
    font {
     bForceMonospaced=false
     bMaterialIsStencil=false
     fontMaterial=null
     fontObject=&Roboto
     letterSpacing=0
     monospacedWidth=1
     size=10.5
     skewAmount=0
     typefaceFontName=Bold
     outlineSettings {
      bApplyOutlineToDropShadows=false
      bMiteredCorners=false
      bSeparateFillAlpha=false
      outlineMaterial=null
      outlineSize=0
      outlineColor=<rgba 0 0 0 1>
     }
    }
   }
   slot {class=HorizontalBoxSlot, properties {verticalAlignment=VAlign_Center, padding=<ltrb 10 0 0 0>}}
  }
  Scan_Btn {
   widgetClassPath=&EditorUtilityButton
   parent=ActiveSequence_HB
   widget=Scan_Btn
   properties {toolTipText=Find all actors with Character Parts in the active LevelSequence}
   slot {class=HorizontalBoxSlot, properties {horizontalAlignment=HAlign_Right, size {sizeRule=Fill, value=1}}}
  }
  TextBlock_340 {
   widgetClassPath=&TextBlock
   parent=Scan_Btn
   widget=TextBlock_340
   properties {
    text=Scan Current LevelSequence
    colorAndOpacity {colorUseRule=UseColor_Specified, specifiedColor=<rgba 0.9698 0.9698 0.9698 1>}
    font {
     bForceMonospaced=false
     bMaterialIsStencil=false
     fontMaterial=null
     fontObject=&Roboto
     letterSpacing=0
     monospacedWidth=1
     size=10.5
     skewAmount=0
     typefaceFontName=Regular
     outlineSettings {
      bApplyOutlineToDropShadows=false
      bMiteredCorners=false
      bSeparateFillAlpha=false
      outlineMaterial=null
      outlineSize=0
      outlineColor=<rgba 0 0 0 1>
     }
    }
   }
   slot {class=ButtonSlot}
  }
  EditorUtilityButton_194 {
   widgetClassPath=&EditorUtilityButton
   parent=ActiveSequence_HB
   widget=EditorUtilityButton_194
   properties {
    toolTipText=Help button to confluence page
    widgetStyle {
     clickedSlateSound {resourceObject=null}
     disabled {
      bIsDynamicallyLoaded=false
      drawAs=RoundedBox
      imageType=NoImage
      mirroring=NoMirror
      resourceName=None
      resourceObject=null
      tiling=NoTile
      imageSize=<0 0>
      margin=<ltrb 0 0 0 0>
      outlineSettings {
       bUseBrushTransparency=false
       roundingType=FixedRadius
       width=1
       color {colorUseRule=UseColor_ColorTable, specifiedColor=<rgba 0 0 0 0>}
       cornerRadii=<4 4 4 4>
      }
      tintColor {colorUseRule=UseColor_Specified, specifiedColor=<rgba 0.0395 0.0395 0.0395 1>}
      uVRegion {bIsValid=false, max=<0 0>, min=<0 0>}
     }
     disabledForeground {colorUseRule=UseColor_ColorTable, specifiedColor=<rgba 0 0 0 0>}
     hovered {
      bIsDynamicallyLoaded=false
      drawAs=RoundedBox
      imageType=FullColor
      mirroring=NoMirror
      resourceName=None
      resourceObject=&/Help_Icon
      tiling=NoTile
      imageSize=<25 25>
      margin=<ltrb 0 0 0 0>
      outlineSettings {
       bUseBrushTransparency=false
       roundingType=FixedRadius
       width=1
       color {colorUseRule=UseColor_ColorTable, specifiedColor=<rgba 0 0 0 0>}
       cornerRadii=<4 4 4 4>
      }
      tintColor {colorUseRule=UseColor_Specified, specifiedColor=<rgba 0.1615 0.1615 0.1615 1>}
      uVRegion {bIsValid=false, max=<0 0>, min=<0 0>}
     }
     hoveredForeground {colorUseRule=UseColor_ColorTable, specifiedColor=<rgba 0 0 0 0>}
     hoveredSlateSound {resourceObject=null}
     normal {
      bIsDynamicallyLoaded=false
      drawAs=RoundedBox
      imageType=FullColor
      mirroring=NoMirror
      resourceName=None
      resourceObject=&Help_Icon
      tiling=NoTile
      imageSize=<25 25>
      margin=<ltrb 0 0 0 0>
      outlineSettings {
       bUseBrushTransparency=false
       roundingType=FixedRadius
       width=1
       color {colorUseRule=UseColor_ColorTable, specifiedColor=<rgba 0 0 0 0>}
       cornerRadii=<4 4 4 4>
      }
      tintColor {colorUseRule=UseColor_Specified, specifiedColor=<rgba 1 1 1 1>}
      uVRegion {bIsValid=false, max=<0 0>, min=<0 0>}
     }
     normalForeground {colorUseRule=UseColor_ColorTable, specifiedColor=<rgba 0 0 0 0>}
     normalPadding=<ltrb 12 1.5 12 1.5>
     pressed {
      bIsDynamicallyLoaded=false
      drawAs=RoundedBox
      imageType=FullColor
      mirroring=NoMirror
      resourceName=None
      resourceObject=&Help_Icon
      tiling=NoTile
      imageSize=<25 25>
      margin=<ltrb 0 0 0 0>
      outlineSettings {
       bUseBrushTransparency=false
       roundingType=FixedRadius
       width=1
       color {colorUseRule=UseColor_ColorTable, specifiedColor=<rgba 0 0 0 0>}
       cornerRadii=<4 4 4 4>
      }
      tintColor {colorUseRule=UseColor_Specified, specifiedColor=<rgba 0.0284 0.0284 0.0284 1>}
      uVRegion {bIsValid=false, max=<0 0>, min=<0 0>}
     }
     pressedForeground {colorUseRule=UseColor_ColorTable, specifiedColor=<rgba 0 0 0 0>}
     pressedPadding=<ltrb 12 2.5 12 0.5>
     pressedSlateSound {resourceObject=null}
    }
   }
   slot {class=HorizontalBoxSlot, properties {padding=<ltrb 5 0 5 0>}}
  }
  HorizontalBox_1 {
   widgetClassPath=&HorizontalBox
   parent=VerticalBox_0
   widget=HorizontalBox_1
   slot {class=VerticalBoxSlot, properties {padding=<ltrb 0 0 0 10>, size {sizeRule=Fill, value=1}}}
  }
  TextBlock_3 {
   widgetClassPath=&TextBlock
   parent=HorizontalBox_1
   widget=TextBlock_3
   properties {
    text="Selected Actor with Character Parts: "
    font {
     bForceMonospaced=false
     bMaterialIsStencil=false
     fontMaterial=null
     fontObject=&Roboto
     letterSpacing=0
     monospacedWidth=1
     size=10.5
     skewAmount=0
     typefaceFontName=Regular
     outlineSettings {
      bApplyOutlineToDropShadows=false
      bMiteredCorners=false
      bSeparateFillAlpha=false
      outlineMaterial=null
      outlineSize=0
      outlineColor=<rgba 0 0 0 1>
     }
    }
   }
   slot {class=HorizontalBoxSlot, properties {verticalAlignment=VAlign_Center}}
  }
  Actor_with_CharacterParts_ComboBox {
   widgetClassPath=/Script/FortniteEditor.SearchableStringComboBox
   bIsVariable=true
   parent=HorizontalBox_1
   widget=Actor_with_CharacterParts_ComboBox
   properties {
    font {
     bForceMonospaced=false
     bMaterialIsStencil=false
     fontMaterial=null
     fontObject=&Roboto
     letterSpacing=0
     monospacedWidth=1
     size=10.5
     skewAmount=0
     typefaceFontName=Regular
     outlineSettings {
      bApplyOutlineToDropShadows=false
      bMiteredCorners=false
      bSeparateFillAlpha=false
      outlineMaterial=null
      outlineSize=0
      outlineColor=<rgba 0 0 0 1>
     }
    }
    foregroundColor {colorUseRule=UseColor_Specified, specifiedColor=<rgba 0.5271 0.5271 0.5271 1>}
    itemStyle {
     bUseParentRowBrush=false
     activeBrush {
      bIsDynamicallyLoaded=false
      drawAs=Image
      imageType=NoImage
      mirroring=NoMirror
      resourceName=None
      resourceObject=null
      tiling=NoTile
      imageSize=<0 0>
      margin=<ltrb 0 0 0 0>
      outlineSettings {
       bUseBrushTransparency=false
       roundingType=HalfHeightRadius
       width=0
       color {colorUseRule=UseColor_Specified, specifiedColor=<rgba 0 0 0 0>}
       cornerRadii=<0 0 0 0>
      }
      tintColor {colorUseRule=UseColor_Specified, specifiedColor=<rgba 0 0.162 0.7454 1>}
      uVRegion {bIsValid=false, max=<0 0>, min=<0 0>}
     }
     activeHighlightedBrush {
      bIsDynamicallyLoaded=false
      drawAs=Image
      imageType=NoImage
      mirroring=NoMirror
      resourceName=None
      resourceObject=null
      tiling=NoTile
      imageSize=<0 0>
      margin=<ltrb 0 0 0 0>
      outlineSettings {
       bUseBrushTransparency=false
       roundingType=HalfHeightRadius
       width=0
       color {colorUseRule=UseColor_Specified, specifiedColor=<rgba 0 0 0 0>}
       cornerRadii=<0 0 0 0>
      }
      tintColor {colorUseRule=UseColor_Specified, specifiedColor=<rgba 0.0252 0.0319 0.0423 1>}
      uVRegion {bIsValid=false, max=<0 0>, min=<0 0>}
     }
     activeHoveredBrush {
      bIsDynamicallyLoaded=false
      drawAs=Image
      imageType=NoImage
      mirroring=NoMirror
      resourceName=None
      resourceObject=null
      tiling=NoTile
      imageSize=<0 0>
      margin=<ltrb 0 0 0 0>
      outlineSettings {
       bUseBrushTransparency=false
       roundingType=HalfHeightRadius
       width=0
       color {colorUseRule=UseColor_Specified, specifiedColor=<rgba 0 0 0 0>}
       cornerRadii=<0 0 0 0>
      }
      tintColor {colorUseRule=UseColor_Specified, specifiedColor=<rgba 0 0.162 0.7454 1>}
      uVRegion {bIsValid=false, max=<0 0>, min=<0 0>}
     }
     dropIndicator_Above {
      bIsDynamicallyLoaded=false
      drawAs=Box
      imageType=FullColor
      mirroring=NoMirror
      resourceName=../../../Engine/Content/Slate/Common/DropZoneIndicator_Above.png
      resourceObject=null
      tiling=NoTile
      imageSize=<0 0>
      margin=<ltrb 0.625 0.625 0 0>
      outlineSettings {
       bUseBrushTransparency=false
       roundingType=HalfHeightRadius
       width=0
       color {colorUseRule=UseColor_Specified, specifiedColor=<rgba 0 0 0 0>}
       cornerRadii=<0 0 0 0>
      }
      tintColor {colorUseRule=UseColor_Specified, specifiedColor=<rgba 0.0091 0.3515 0.9647 1>}
      uVRegion {bIsValid=false, max=<0 0>, min=<0 0>}
     }
     dropIndicator_Below {
      bIsDynamicallyLoaded=false
      drawAs=Box
      imageType=FullColor
      mirroring=NoMirror
      resourceName=../../../Engine/Content/Slate/Common/DropZoneIndicator_Below.png
      resourceObject=null
      tiling=NoTile
      imageSize=<0 0>
      margin=<ltrb 0.625 0 0 0.625>
      outlineSettings {
       bUseBrushTransparency=false
       roundingType=HalfHeightRadius
       width=0
       color {colorUseRule=UseColor_Specified, specifiedColor=<rgba 0 0 0 0>}
       cornerRadii=<0 0 0 0>
      }
      tintColor {colorUseRule=UseColor_Specified, specifiedColor=<rgba 0.0091 0.3515 0.9647 1>}
      uVRegion {bIsValid=false, max=<0 0>, min=<0 0>}
     }
     dropIndicator_Onto {
      bIsDynamicallyLoaded=false
      drawAs=Box
      imageType=FullColor
      mirroring=NoMirror
      resourceName=../../../Engine/Content/Slate/Common/DropZoneIndicator_Onto.png
      resourceObject=null
      tiling=NoTile
      imageSize=<0 0>
      margin=<ltrb 0.25 0.25 0.25 0.25>
      outlineSettings {
       bUseBrushTransparency=false
       roundingType=HalfHeightRadius
       width=0
       color {colorUseRule=UseColor_Specified, specifiedColor=<rgba 0 0 0 0>}
       cornerRadii=<0 0 0 0>
      }
      tintColor {colorUseRule=UseColor_Specified, specifiedColor=<rgba 0.0091 0.3515 0.9647 1>}
      uVRegion {bIsValid=false, max=<0 0>, min=<0 0>}
     }
     evenRowBackgroundBrush {
      bIsDynamicallyLoaded=false
      drawAs=Image
      imageType=NoImage
      mirroring=NoMirror
      resourceName=None
      resourceObject=null
      tiling=NoTile
      imageSize=<0 0>
      margin=<ltrb 0 0 0 0>
      outlineSettings {
       bUseBrushTransparency=false
       roundingType=HalfHeightRadius
       width=0
       color {colorUseRule=UseColor_Specified, specifiedColor=<rgba 0 0 0 0>}
       cornerRadii=<0 0 0 0>
      }
      tintColor {colorUseRule=UseColor_Specified, specifiedColor=<rgba 0.0103 0.0103 0.0103 1>}
      uVRegion {bIsValid=false, max=<0 0>, min=<0 0>}
     }
     evenRowBackgroundHoveredBrush {
      bIsDynamicallyLoaded=false
      drawAs=Image
      imageType=NoImage
      mirroring=NoMirror
      resourceName=None
      resourceObject=null
      tiling=NoTile
      imageSize=<0 0>
      margin=<ltrb 0 0 0 0>
      outlineSettings {
       bUseBrushTransparency=false
       roundingType=HalfHeightRadius
       width=0
       color {colorUseRule=UseColor_Specified, specifiedColor=<rgba 0 0 0 0>}
       cornerRadii=<0 0 0 0>
      }
      tintColor {colorUseRule=UseColor_Specified, specifiedColor=<rgba 0.0176 0.0176 0.0176 1>}
      uVRegion {bIsValid=false, max=<0 0>, min=<0 0>}
     }
     inactiveBrush {
      bIsDynamicallyLoaded=false
      drawAs=Image
      imageType=NoImage
      mirroring=NoMirror
      resourceName=None
      resourceObject=null
      tiling=NoTile
      imageSize=<0 0>
      margin=<ltrb 0 0 0 0>
      outlineSettings {
       bUseBrushTransparency=false
       roundingType=HalfHeightRadius
       width=0
       color {colorUseRule=UseColor_Specified, specifiedColor=<rgba 0 0 0 0>}
       cornerRadii=<0 0 0 0>
      }
      tintColor {colorUseRule=UseColor_Specified, specifiedColor=<rgba 0.0513 0.0953 0.159 1>}
      uVRegion {bIsValid=false, max=<0 0>, min=<0 0>}
     }
     inactiveHighlightedBrush {
      bIsDynamicallyLoaded=false
      drawAs=Image
      imageType=NoImage
      mirroring=NoMirror
      resourceName=None
      resourceObject=null
      tiling=NoTile
      imageSize=<0 0>
      margin=<ltrb 0 0 0 0>
      outlineSettings {
       bUseBrushTransparency=false
       roundingType=HalfHeightRadius
       width=0
       color {colorUseRule=UseColor_Specified, specifiedColor=<rgba 0 0 0 0>}
       cornerRadii=<0 0 0 0>
      }
      tintColor {colorUseRule=UseColor_Specified, specifiedColor=<rgba 0.0252 0.0319 0.0423 1>}
      uVRegion {bIsValid=false, max=<0 0>, min=<0 0>}
     }
     inactiveHoveredBrush {
      bIsDynamicallyLoaded=false
      drawAs=Image
      imageType=NoImage
      mirroring=NoMirror
      resourceName=None
      resourceObject=null
      tiling=NoTile
      imageSize=<0 0>
      margin=<ltrb 0 0 0 0>
      outlineSettings {
       bUseBrushTransparency=false
       roundingType=HalfHeightRadius
       width=0
       color {colorUseRule=UseColor_Specified, specifiedColor=<rgba 0 0 0 0>}
       cornerRadii=<0 0 0 0>
      }
      tintColor {colorUseRule=UseColor_Specified, specifiedColor=<rgba 0.0513 0.0953 0.159 1>}
      uVRegion {bIsValid=false, max=<0 0>, min=<0 0>}
     }
     oddRowBackgroundBrush {
      bIsDynamicallyLoaded=false
      drawAs=Image
      imageType=NoImage
      mirroring=NoMirror
      resourceName=None
      resourceObject=null
      tiling=NoTile
      imageSize=<0 0>
      margin=<ltrb 0 0 0 0>
      outlineSettings {
       bUseBrushTransparency=false
       roundingType=HalfHeightRadius
       width=0
       color {colorUseRule=UseColor_Specified, specifiedColor=<rgba 0 0 0 0>}
       cornerRadii=<0 0 0 0>
      }
      tintColor {colorUseRule=UseColor_Specified, specifiedColor=<rgba 0.0103 0.0103 0.0103 1>}
      uVRegion {bIsValid=false, max=<0 0>, min=<0 0>}
     }
     oddRowBackgroundHoveredBrush {
      bIsDynamicallyLoaded=false
      drawAs=Image
      imageType=NoImage
      mirroring=NoMirror
      resourceName=None
      resourceObject=null
      tiling=NoTile
      imageSize=<0 0>
      margin=<ltrb 0 0 0 0>
      outlineSettings {
       bUseBrushTransparency=false
       roundingType=HalfHeightRadius
       width=0
       color {colorUseRule=UseColor_Specified, specifiedColor=<rgba 0 0 0 0>}
       cornerRadii=<0 0 0 0>
      }
      tintColor {colorUseRule=UseColor_Specified, specifiedColor=<rgba 0.0176 0.0176 0.0176 1>}
      uVRegion {bIsValid=false, max=<0 0>, min=<0 0>}
     }
     parentRowBackgroundBrush {
      bIsDynamicallyLoaded=false
      drawAs=Image
      imageType=NoImage
      mirroring=NoMirror
      resourceName=None
      resourceObject=null
      tiling=NoTile
      imageSize=<32 32>
      margin=<ltrb 0 0 0 0>
      outlineSettings {
       bUseBrushTransparency=false
       roundingType=HalfHeightRadius
       width=0
       color {colorUseRule=UseColor_Specified, specifiedColor=<rgba 0 0 0 0>}
       cornerRadii=<0 0 0 0>
      }
      tintColor {colorUseRule=UseColor_Specified, specifiedColor=<rgba 1 1 1 1>}
      uVRegion {bIsValid=false, max=<0 0>, min=<0 0>}
     }
     parentRowBackgroundHoveredBrush {
      bIsDynamicallyLoaded=false
      drawAs=Image
      imageType=NoImage
      mirroring=NoMirror
      resourceName=None
      resourceObject=null
      tiling=NoTile
      imageSize=<32 32>
      margin=<ltrb 0 0 0 0>
      outlineSettings {
       bUseBrushTransparency=false
       roundingType=HalfHeightRadius
       width=0
       color {colorUseRule=UseColor_Specified, specifiedColor=<rgba 0 0 0 0>}
       cornerRadii=<0 0 0 0>
      }
      tintColor {colorUseRule=UseColor_Specified, specifiedColor=<rgba 1 1 1 1>}
      uVRegion {bIsValid=false, max=<0 0>, min=<0 0>}
     }
     selectedTextColor {colorUseRule=UseColor_Specified, specifiedColor=<rgba 1 1 1 1>}
     selectorFocusedBrush {
      bIsDynamicallyLoaded=false
      drawAs=Border
      imageType=FullColor
      mirroring=NoMirror
      resourceName=../../../Engine/Content/Slate/Common/Selector.png
      resourceObject=null
      tiling=Both
      imageSize=<0 0>
      margin=<ltrb 0.25 0.25 0.25 0.25>
      outlineSettings {
       bUseBrushTransparency=false
       roundingType=HalfHeightRadius
       width=0
       color {colorUseRule=UseColor_Specified, specifiedColor=<rgba 0 0 0 0>}
       cornerRadii=<0 0 0 0>
      }
      tintColor {colorUseRule=UseColor_Specified, specifiedColor=<rgba 0.701 0.225 0.003 1>}
      uVRegion {bIsValid=false, max=<0 0>, min=<0 0>}
     }
     textColor {colorUseRule=UseColor_Specified, specifiedColor=<rgba 0.5271 0.5271 0.5271 1>}
    }
    widgetStyle {
     comboButtonStyle {
      downArrowAlign=VAlign_Center
      buttonStyle {
       clickedSlateSound {resourceObject=null}
       disabled {
        bIsDynamicallyLoaded=false
        drawAs=RoundedBox
        imageType=NoImage
        mirroring=NoMirror
        resourceName=None
        resourceObject=null
        tiling=NoTile
        imageSize=<0 0>
        margin=<ltrb 0 0 0 0>
        outlineSettings {
         bUseBrushTransparency=false
         roundingType=FixedRadius
         width=1
         color {colorUseRule=UseColor_ColorTable, specifiedColor=<rgba 0 0 0 0>}
         cornerRadii=<4 4 4 4>
        }
        tintColor {colorUseRule=UseColor_Specified, specifiedColor=<rgba 0.0395 0.0395 0.0395 1>}
        uVRegion {bIsValid=false, max=<0 0>, min=<0 0>}
       }
       disabledForeground {colorUseRule=UseColor_ColorTable, specifiedColor=<rgba 0 0 0 0>}
       hovered {
        bIsDynamicallyLoaded=false
        drawAs=RoundedBox
        imageType=NoImage
        mirroring=NoMirror
        resourceName=None
        resourceObject=null
        tiling=NoTile
        imageSize=<0 0>
        margin=<ltrb 0 0 0 0>
        outlineSettings {
         bUseBrushTransparency=false
         roundingType=FixedRadius
         width=1
         color {colorUseRule=UseColor_ColorTable, specifiedColor=<rgba 0 0 0 0>}
         cornerRadii=<4 4 4 4>
        }
        tintColor {colorUseRule=UseColor_Specified, specifiedColor=<rgba 0.0048 0.0048 0.0048 1>}
        uVRegion {bIsValid=false, max=<0 0>, min=<0 0>}
       }
       hoveredForeground {colorUseRule=UseColor_ColorTable, specifiedColor=<rgba 0 0 0 0>}
       hoveredSlateSound {resourceObject=null}
       normal {
        bIsDynamicallyLoaded=false
        drawAs=RoundedBox
        imageType=NoImage
        mirroring=NoMirror
        resourceName=None
        resourceObject=null
        tiling=NoTile
        imageSize=<0 0>
        margin=<ltrb 0 0 0 0>
        outlineSettings {
         bUseBrushTransparency=false
         roundingType=FixedRadius
         width=1
         color {colorUseRule=UseColor_ColorTable, specifiedColor=<rgba 0 0 0 0>}
         cornerRadii=<4 4 4 4>
        }
        tintColor {colorUseRule=UseColor_Specified, specifiedColor=<rgba 0.0048 0.0048 0.0048 1>}
        uVRegion {bIsValid=false, max=<0 0>, min=<0 0>}
       }
       normalForeground {colorUseRule=UseColor_ColorTable, specifiedColor=<rgba 0 0 0 0>}
       normalPadding=<ltrb 8 1 8 1>
       pressed {
        bIsDynamicallyLoaded=false
        drawAs=RoundedBox
        imageType=NoImage
        mirroring=NoMirror
        resourceName=None
        resourceObject=null
        tiling=NoTile
        imageSize=<0 0>
        margin=<ltrb 0 0 0 0>
        outlineSettings {
         bUseBrushTransparency=false
         roundingType=FixedRadius
         width=1
         color {colorUseRule=UseColor_ColorTable, specifiedColor=<rgba 0 0 0 0>}
         cornerRadii=<4 4 4 4>
        }
        tintColor {colorUseRule=UseColor_Specified, specifiedColor=<rgba 0.0048 0.0048 0.0048 1>}
        uVRegion {bIsValid=false, max=<0 0>, min=<0 0>}
       }
       pressedForeground {colorUseRule=UseColor_ColorTable, specifiedColor=<rgba 0 0 0 0>}
       pressedPadding=<ltrb 8 2 8 0>
       pressedSlateSound {resourceObject=null}
      }
      contentPadding=<ltrb 0 0 0 0>
      downArrowImage {
       bIsDynamicallyLoaded=false
       drawAs=Image
       imageType=Vector
       mirroring=NoMirror
       resourceName=../../../Engine/Content/Slate/Starship/CoreWidgets/ComboBox/wide-chevron-down.svg
       resourceObject=null
       tiling=NoTile
       imageSize=<16 16>
       margin=<ltrb 0 0 0 0>
       outlineSettings {
        bUseBrushTransparency=false
        roundingType=HalfHeightRadius
        width=0
        color {colorUseRule=UseColor_Specified, specifiedColor=<rgba 0 0 0 0>}
        cornerRadii=<0 0 0 0>
       }
       tintColor {colorUseRule=UseColor_Specified, specifiedColor=<rgba 1 1 1 1>}
       uVRegion {bIsValid=false, max=<0 0>, min=<0 0>}
      }
      downArrowPadding=<ltrb 2 2 2 2>
      menuBorderBrush {
       bIsDynamicallyLoaded=false
       drawAs=RoundedBox
       imageType=NoImage
       mirroring=NoMirror
       resourceName=None
       resourceObject=null
       tiling=NoTile
       imageSize=<0 0>
       margin=<ltrb 0 0 0 0>
       outlineSettings {
        bUseBrushTransparency=false
        roundingType=FixedRadius
        width=1
        color {colorUseRule=UseColor_ColorTable, specifiedColor=<rgba 0 0 0 0>}
        cornerRadii=<0 0 0 0>
       }
       tintColor {colorUseRule=UseColor_Specified, specifiedColor=<rgba 0.0395 0.0395 0.0395 1>}
       uVRegion {bIsValid=false, max=<0 0>, min=<0 0>}
      }
      menuBorderPadding=<ltrb 0 0 0 0>
      shadowColorAndOpacity=<rgba 0 0 0 1>
      shadowOffset=<0 0>
     }
     contentPadding=<ltrb 0 0 0 0>
     menuRowPadding=<ltrb 8 3 8 3>
     pressedSlateSound {resourceObject=null}
     selectionChangeSlateSound {resourceObject=null}
    }
   }
   slot {class=HorizontalBoxSlot}
  }
  Border_1 {
   widgetClassPath=&Border
   parent=VerticalBox_0
   widget=Border_1
   properties {background {
     bIsDynamicallyLoaded=false
     drawAs=Image
     imageType=NoImage
     mirroring=NoMirror
     resourceName=None
     resourceObject=null
     tiling=NoTile
     imageSize=<32 32>
     margin=<ltrb 0 0 0 0>
     outlineSettings {
      bUseBrushTransparency=false
      roundingType=HalfHeightRadius
      width=0
      color {colorUseRule=UseColor_Specified, specifiedColor=<rgba 0 0 0 0>}
      cornerRadii=<0 0 0 0>
     }
     tintColor {colorUseRule=UseColor_Specified, specifiedColor=<rgba 0.0075 0.0075 0.0075 1>}
     uVRegion {bIsValid=false, max=<0 0>, min=<0 0>}
    }}
   slot {class=VerticalBoxSlot, properties {size {sizeRule=Fill, value=1}}}
  }
  HorizontalBox_193 {widgetClassPath=&HorizontalBox, parent=Border_1, widget=HorizontalBox_193, slot {class=BorderSlot}}
  TextBlock_4 {
   widgetClassPath=&TextBlock
   parent=HorizontalBox_193
   widget=TextBlock_4
   properties {
    text=Slot
    font {
     bForceMonospaced=false
     bMaterialIsStencil=false
     fontMaterial=null
     fontObject=&Roboto
     letterSpacing=0
     monospacedWidth=1
     size=10.5
     skewAmount=0
     typefaceFontName=Bold
     outlineSettings {
      bApplyOutlineToDropShadows=false
      bMiteredCorners=false
      bSeparateFillAlpha=false
      outlineMaterial=null
      outlineSize=0
      outlineColor=<rgba 0 0 0 1>
     }
    }
   }
   slot {class=HorizontalBoxSlot, properties {verticalAlignment=VAlign_Center, padding=<ltrb 0 4 8 4>, size {sizeRule=Fill, value=0.18}}}
  }
  TextBlock_5 {
   widgetClassPath=&TextBlock
   parent=HorizontalBox_193
   widget=TextBlock_5
   properties {
    text=Cosmetic Asset
    font {
     bForceMonospaced=false
     bMaterialIsStencil=false
     fontMaterial=null
     fontObject=&Roboto
     letterSpacing=0
     monospacedWidth=1
     size=10.5
     skewAmount=0
     typefaceFontName=Bold
     outlineSettings {
      bApplyOutlineToDropShadows=false
      bMiteredCorners=false
      bSeparateFillAlpha=false
      outlineMaterial=null
      outlineSize=0
      outlineColor=<rgba 0 0 0 1>
     }
    }
   }
   slot {class=HorizontalBoxSlot, properties {verticalAlignment=VAlign_Center, padding=<ltrb 8 4 8 4>, size {sizeRule=Fill, value=0.42}}}
  }
  TextBlock_6 {
   widgetClassPath=&TextBlock
   parent=HorizontalBox_193
   widget=TextBlock_6
   properties {
    text=ControlRig
    font {
     bForceMonospaced=false
     bMaterialIsStencil=false
     fontMaterial=null
     fontObject=&Roboto
     letterSpacing=0
     monospacedWidth=1
     size=10.5
     skewAmount=0
     typefaceFontName=Bold
     outlineSettings {
      bApplyOutlineToDropShadows=false
      bMiteredCorners=false
      bSeparateFillAlpha=false
      outlineMaterial=null
      outlineSize=0
      outlineColor=<rgba 0 0 0 1>
     }
    }
   }
   slot {class=HorizontalBoxSlot, properties {verticalAlignment=VAlign_Center, padding=<ltrb 8 4 8 4>, size {sizeRule=Fill, value=0.38}}}
  }
  Setup_ScrollBox {
   widgetClassPath=&EditorUtilityScrollBox
   bIsVariable=true
   parent=VerticalBox_0
   widget=Setup_ScrollBox
   slot {class=VerticalBoxSlot}
  }
  WBP_CPRow {
   widgetClassPath=&/WBP_CPRow
   bIsVariable=true
   parent=Setup_ScrollBox
   widget=WBP_CPRow
   slot {class=ScrollBoxSlot}
  }
  NoParts_Text {
   widgetClassPath=&TextBlock
   bIsVariable=true
   parent=VerticalBox_0
   widget=NoParts_Text
   properties {
    text=No character parts found in scene
    visibility=Hidden
    font {
     bForceMonospaced=false
     bMaterialIsStencil=false
     fontMaterial=null
     fontObject=&Roboto
     letterSpacing=0
     monospacedWidth=1
     size=24
     skewAmount=0
     typefaceFontName=Regular
     outlineSettings {
      bApplyOutlineToDropShadows=false
      bMiteredCorners=false
      bSeparateFillAlpha=false
      outlineMaterial=null
      outlineSize=0
      outlineColor=<rgba 0 0 0 1>
     }
    }
   }
   slot {class=VerticalBoxSlot}
  }
  Setup_Button {
   widgetClassPath=&EditorUtilityButton
   parent=ControlRig
   widget=Setup_Button
   properties {toolTipText=Apply character part changes and set up rig tracks in the active Level Sequence}
   slot {
    class=VerticalBoxSlot
    properties {horizontalAlignment=HAlign_Center, verticalAlignment=VAlign_Center, padding=<ltrb 0 10 0 10>, size {sizeRule=Fill, value=1}}
   }
  }
  TextBlock_263 {
   widgetClassPath=&TextBlock
   parent=Setup_Button
   widget=TextBlock_263
   properties {
    text="Apply Character Parts && Rigs"
    font {
     bForceMonospaced=false
     bMaterialIsStencil=false
     fontMaterial=null
     fontObject=&Roboto
     letterSpacing=0
     monospacedWidth=1
     size=15
     skewAmount=0
     typefaceFontName=Bold
     outlineSettings {
      bApplyOutlineToDropShadows=false
      bMiteredCorners=false
      bSeparateFillAlpha=false
      outlineMaterial=null
      outlineSize=0
      outlineColor=<rgba 0 0 0 1>
     }
    }
   }
   slot {class=ButtonSlot}
  }
 ]
}
